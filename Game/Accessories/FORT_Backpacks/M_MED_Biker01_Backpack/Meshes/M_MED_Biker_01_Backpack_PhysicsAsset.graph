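# Physics asset for the M_MED_Biker01 backpack accessory's skeletal mesh.
# It holds a single capsule body on the backpack_root bone, offset and rotated from the bone origin to form a simple collision volume.
# It has no constraints, so no joints link any simulated parts.
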PhysicsAsset /Game/Accessories/FORT_Backpacks/M_MED_Biker01_Backpack/Meshes/M_MED_Biker_01_Backpack_PhysicsAsset
bodies {backpack_root [{shapeType=Capsule, length=10.5286, radius=11.1027, center=<-12.3969 -1.9293 0>, rotation=<rot -85.4037 -85.404 90.0007>}]}
constraints []
